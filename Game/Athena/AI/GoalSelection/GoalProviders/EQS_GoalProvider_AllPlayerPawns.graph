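# EnvQuery goal provider for Athena AI goal selection that gathers player pawns as candidate goals.
# It has a single option using FortQueryGenerator_GoalPlayerPawns, restricted to Athena game participants and not run asynchronously.
# The option has no tests, so the query itself does no filtering or scoring beyond what the generator returns.
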EnvQuery /Game/Athena/AI/GoalSelection/GoalProviders/EQS_GoalProvider_AllPlayerPawns
options [{
  generator {
   name=FortQueryGenerator_GoalPlayerPawns_0
   class=FortQueryGenerator_GoalPlayerPawns
   properties {bAutoSortTests=true, bCanRunAsync=false, bOnlyAthenaGameParticipants=true, envQueryResultNormalizationOption=Default, optionName=""}
  }
  tests []
 }]
